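; Ferris wheel voice-over audio gate Blueprint that fires directional enter and exit cues as an audio gameplay volume toggles.
; It infers the crossing side from the sign of the dot product between the root's forward vector and the first local player pawn's normalized movement velocity, checked only when that pawn is locally viewed.
; That boolean selects the front or back trigger parameter executed on the Audio component.
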(event OnToggledOff(AudioGameplayVolumeToggle)
  ;[0.5751 0.1986 0.8854 0.5] Exiting Either side of the volume
  (bind _returnvalue (CallFunction|GetSidePlayerIntersected))
  (bind _audio (Variables|Default|GetAudio)
    (:then
      (Audio|Parameter|ExecuteTriggerParameter _audio (Utilities|String|SelectName "Play.Back.Exit" "Play.Front.Exit" _returnvalue)))
    (:else)))

(event OnToggledOn(AudioGameplayVolumeToggle)
  (bind _returnvalue (CallFunction|GetSidePlayerIntersected))
  (bind _audio (Variables|Default|GetAudio)
    (:then
      (Audio|Parameter|ExecuteTriggerParameter _audio (Utilities|String|SelectName "Play.Front.Enter" "Play.Back.Enter" _returnvalue)))
    (:else)))

(fn GetSidePlayerIntersected
  (bind _returnvalue (Fortnite|GetFirstLocalFortPlayerController true))
  (bind _returnvalue_2 (Game|Player|GetPlayerPawn _returnvalue))
  (Utilities|IsValid _returnvalue
    (:Is%20Valid
      (Utilities|IsValid _returnvalue_2
        (:Is%20Valid
          (if (Pawn|IsLocallyViewed _returnvalue_2)
            (return (> (Math|Vector|DotProduct (Transformation|GetForwardVector (Variables|Default|GetDefaultSceneRoot)) (Math|Vector|Normalize (Variables|/Script/Engine.MovementComponent|GetVelocity (Variables|/Script/Engine.Character|GetCharacterMovement _returnvalue_2)))) 0))))))))

(fn ConstructionScript)
