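# Looping mono MetaSound source for HoneyFiber Ferris wheel NPC voice lines, with random ambient barks and exit reactions.
# Ambient lines start after a short random delay, then recur with longer random gaps.
# Play.Front.Exit retries through a gate until no ambient line is playing, then plays a random exit line, resetting ambient timers until it finishes.
# Trigger variables close these feedback loops; VO.Fade.Out and VO.KeepAlive are unconnected.
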MetaSoundSource /HoneyFiberEnv/Audio/VO/FerrisWheel/IP_MSS_VO_HoneyFibre_FerrisWheel
inputs [
 Ambient {type=WaveAsset:Array}
 Front.Exit {type=WaveAsset:Array}
 Play.Front.Exit {type=Trigger}
 UE.Source.OnPlay {type=Trigger}
 VO.Fade.Out {type=Trigger}
 VO.KeepAlive {type=Trigger}
]
interfaces [UE.OutputFormat.Mono:1.0, UE.Source:1.1]
outputs ["UE.OutputFormat.Mono.Audio:0" {type=Audio}]
pages [Default {
  input_defaults {
   Play.Front.Exit=null
   UE.Source.OnPlay=null
   VO.Fade.Out=null
   VO.KeepAlive=null
   Ambient [
    /HoneyFiberEnv/Audio/VO/FerrisWheel/Waves/IP_VO_HoneyFiber_FerrisWheel_Ambient_Authoritah_01
    /HoneyFiberEnv/Audio/VO/FerrisWheel/Waves/IP_VO_HoneyFiber_FerrisWheel_Ambient_Authoritah_02
    /HoneyFiberEnv/Audio/VO/FerrisWheel/Waves/IP_VO_HoneyFiber_FerrisWheel_Ambient_Beefcake
    /HoneyFiberEnv/Audio/VO/FerrisWheel/Waves/IP_VO_HoneyFiber_FerrisWheel_Ambient_CheesyPoofs
    /HoneyFiberEnv/Audio/VO/FerrisWheel/Waves/IP_VO_HoneyFiber_FerrisWheel_Ambient_Cool
    /HoneyFiberEnv/Audio/VO/FerrisWheel/Waves/IP_VO_HoneyFiber_FerrisWheel_Ambient_NotFat
   ]
   Front.Exit [
    /HoneyFiberEnv/Audio/VO/FerrisWheel/Waves/IP_VO_HoneyFiber_FerrisWheel_OnExit_OhDudeLame
    /HoneyFiberEnv/Audio/VO/FerrisWheel/Waves/IP_VO_HoneyFiber_FerrisWheel_OnExit_NotCool
   ]
  }
  nodes [
   n1 = VariableDeferredAccessor.Trigger(Variable=InitVariable.Trigger().Variable)
   TriggerAnyTriggerAny2_1 = "TriggerAny.Trigger Any (2)"(In 0=^UE.Source.OnPlay, In 1=@n1.Value)
   n2 = UE.RandomTime(Next=@TriggerAnyTriggerAny2_1.Out) {also {defaults {Max=3, Min=1}}}
   n3 = InitVariable.Trigger()
   n4 = VariableDeferredAccessor.Trigger(Variable=@n3.Variable)
   TriggerDelay_0 = UE.Trigger Delay(In=@n2.On Next, Delay Time=@n2.Value, Reset=@n4.Value)
   n5 = VariableDeferredAccessor.Trigger(Variable=InitVariable.Trigger().Variable)
   TriggerAnyTriggerAny2_0 = "TriggerAny.Trigger Any (2)"(In 0=@TriggerDelay_0.Out, In 1=@n5.Value)
   ArrayRandomGet_1 = Array.Random Get:WaveAsset:Array(In Array=^Ambient, Next=@TriggerAnyTriggerAny2_0.Out)
   WavePlayer_0 = UE.Wave Player:Mono(Play=@ArrayRandomGet_1.On Next, Wave Asset=@ArrayRandomGet_1.Value, Stop=VariableDeferredAccessor.Trigger(Variable=@n3.Variable).Value) {also {version [1,1]}}
   n6 = VariableDeferredAccessor.Trigger(Variable=@n4.Variable)
   TriggerControl_0 = UE.Trigger Control(Trigger In=UE.TriggerRepeat(Start=^Play.Front.Exit, Stop=@n6.Value).RepeatOut, Open=@WavePlayer_0.On Finished, Close=@WavePlayer_0.On Play)
   ArrayRandomGet_0 = Array.Random Get:WaveAsset:Array(In Array=^Front.Exit, Next=@TriggerControl_0.Trigger Out) {also {defaults {Weights []}}}
   WavePlayer_1 = UE.Wave Player:Mono(Play=@ArrayRandomGet_0.On Next, Wave Asset=@ArrayRandomGet_0.Value) {also {version [1,1]}}
   AudioMixerAudioMixerMono2_0 = "AudioMixer.Audio Mixer (Mono, 2)"(In 0=UE.Reroute(Value=@WavePlayer_0.Out Mono).Value, In 1=@WavePlayer_1.Out Mono)
   n7 = UE.RandomTime(Next=@WavePlayer_0.On Finished) {also {defaults {Max=10, Min=5}}}
   TriggerDelay_1 = UE.Trigger Delay(In=@n7.On Next, Delay Time=@n7.Value, Reset=VariableDeferredAccessor.Trigger(Variable=@n3.Variable).Value)
   VariableMutator.Trigger(Value=@TriggerDelay_1.Out, Variable=@n5.Variable)
   VariableMutator.Trigger(Variable=@n1.Variable, Value=@WavePlayer_1.On Finished)
   VariableMutator.Trigger(Variable=@n6.Variable, Value=@WavePlayer_1.On Play)
  ]
  links [{dst="@UE.OutputFormat.Mono.Audio:0", src=AudioMixerAudioMixerMono2_0.Out}]
 }]
properties {
 assetImportData=&self:AssetImportData
 attenuationSettings=&/IP_ATT_HoneyFibre_FerrisWheel
 bDebug=true
 bLooping=true
 duration=10000
 maxDistance=3200
 sampleRate=48000
 soundClassObject=&Fort_VO_NPC
}
